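# Sound cue for third-person footstep foley of the bush item in Athena, routed to the enemy 3P foley sound class with walk footstep attenuation.
# A Random node picks one of four equally weighted bush footstep layer waves, excluded from branch culling.
# A Modulator above it randomizes volume and raises pitch within a random range.
# A concurrency override limits instances per owner, stopping the oldest when exceeded.
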SoundCue /Game/Athena/Sounds/Bush/Athena_Character_Bush_Footstep_3P_Cue
root=Modulator_0
nodes [
 {id=Modulator_0, children [Random_0], properties {pitchMax=1.5, pitchMin=1.3, volumeMax=1, volumeMin=0.7}}
 {
  id=Random_0
  children [WavePlayer_10, WavePlayer_11, WavePlayer_12, WavePlayer_13]
  properties {bShouldExcludeFromBranchCulling=true, weights [1,1,1,1]}
 }
 {id=WavePlayer_10, properties {soundWaveAssetPtr=&/sfx_item_bush_footstep_layer_01}}
 {id=WavePlayer_11, properties {soundWaveAssetPtr=&/sfx_item_bush_footstep_layer_04}}
 {id=WavePlayer_12, properties {soundWaveAssetPtr=&/sfx_item_bush_footstep_layer_03}}
 {id=WavePlayer_13, properties {soundWaveAssetPtr=&/sfx_item_bush_footstep_layer_02}}
]
properties {
 attenuationSettings=/Game/Sounds/Attenuation/Fort_Player_Footsteps_3P_Walk_Athena_Att
 bOverrideConcurrency=true
 maxDistance=3584
 soundClassObject=&Fort_Player_Foley_3P_Enemy_Athena
 volumeMultiplier=0.7
 concurrencyOverrides {bLimitToOwner=true, maxCount=2, resolutionRule=StopOldest}
}
