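# Footstep foley cue for the Athena character bush, using first-person foley class and crouched footstep attenuation.
# A random node picks one of four equally weighted bush footstep layer waves.
# A modulator above it varies volume and raises pitch well above normal on each play.
# A concurrency override limits playback to two sounds per owner, stopping the oldest.
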SoundCue /Game/Athena/Sounds/Bush/Athena_Character_Bush_Footstep_Cue
root=Modulator_0
nodes [
 {id=Modulator_0, children [Random_0], properties {pitchMax=1.5, pitchMin=1.3, volumeMax=1, volumeMin=0.7}}
 {
  id=Random_0
  children [WavePlayer_6,WavePlayer_7,WavePlayer_8,WavePlayer_9]
  properties {bShouldExcludeFromBranchCulling=true, weights [1,1,1,1]}
 }
 {id=WavePlayer_6, properties {soundWaveAssetPtr=&/sfx_item_bush_footstep_layer_01}}
 {id=WavePlayer_7, properties {soundWaveAssetPtr=&/sfx_item_bush_footstep_layer_04}}
 {id=WavePlayer_8, properties {soundWaveAssetPtr=&/sfx_item_bush_footstep_layer_03}}
 {id=WavePlayer_9, properties {soundWaveAssetPtr=&/sfx_item_bush_footstep_layer_02}}
]
properties {
 attenuationSettings=/Game/Sounds/Attenuation/Fort_Player_Footsteps_1P_Crouch_Athena_Att
 bOverrideConcurrency=true
 maxDistance=2097152
 soundClassObject=&Fort_Player_Foley_1P_Athena
 volumeMultiplier=0.9
 concurrencyOverrides {bLimitToOwner=true, maxCount=2, resolutionRule=StopOldest}
}
